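# Sound cue for a quiet, looping TV static noise placed as a spot sound on screen props in the Helios environment.
# A looping wave player of the TheRaft TV static sample feeds a Modulator at the cue root.
# It uses a spherical screen-static attenuation asset and the gadgets/items sound class, with a concurrency override that keeps simultaneous instances to a small limit.
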SoundCue /Game/Sounds/Helios/Environment_Actors/Screens/Spot_TV_Static_Noise_01_Loop
root=Modulator_0
nodes [
 {id=Modulator_0, children [WavePlayer_0]}
 {id=WavePlayer_0, properties {bLooping=true, soundWaveAssetPtr=&/Spot_TheRaft_TV_Static_01}}
]
properties {
 attenuationSettings=&/Spot_TV_Screen_Static_Sphere_Att
 bOverrideConcurrency=true
 maxDistance=801
 soundClassObject=&Fort_Gadgets_Items
 volumeMultiplier=0.1
 concurrencyOverrides {maxCount=2}
}
