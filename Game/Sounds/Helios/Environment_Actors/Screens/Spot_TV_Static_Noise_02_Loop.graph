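# Looping positional sound cue for TV screen static, filed under Helios environment actor screens.
# A mixer layers a TV static loop, passed through a modulator for pitch variation, with a looping cube shard idle static borrowed from the MASH LTM.
# Overridden attenuation uses a cone shape with distance filtering, occlusion and listener focus, in place of the referenced sphere settings.
# A concurrency override caps it at two instances, and a submix send feeds the default reverb.
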SoundCue /Game/Sounds/Helios/Environment_Actors/Screens/Spot_TV_Static_Noise_02_Loop
root=Mixer_1
nodes [
 {id=Mixer_1, children [Modulator_1,WavePlayer_8], properties {inputVolume [1,1]}}
 {id=Modulator_1, children [WavePlayer_4], properties {pitchMax=1.2}}
 {id=WavePlayer_4, properties {bLooping=true, soundWaveAssetPtr=&/FNM_TV_Static_loop}}
 {id=WavePlayer_8, properties {bLooping=true, soundWaveAssetPtr=/Game/Athena/Sounds/LTMs/MASH/MASH_CubeShard_IdleStatic}}
]
properties {
 attenuationSettings=&/Spot_TV_Screen_Static_Sphere_Att
 bOverrideAttenuation=true
 bOverrideConcurrency=true
 maxDistance=1100
 soundClassObject=&Fort_Gadgets_Items
 volumeMultiplier=0.4
 attenuationOverrides {
  attenuationShape=Cone
  bAttenuateWithLPF=true
  bEnableListenerFocus=true
  bEnableOcclusion=true
  coneSphereFalloffDistance=750
  coneSphereRadius=1
  falloffDistance=850
  hPFFrequencyAtMax=200
  lPFFrequencyAtMax=600
  lPFRadiusMax=1400
  lPFRadiusMin=500
  nonFocusVolumeAttenuation=0.75
  occlusionLowPassFilterFrequency=10000
  occlusionTraceChannel=ECC_GameTraceChannel11
  occlusionVolumeAttenuation=0.75
  stereoSpread=0
  attenuationShapeExtents=<250 90>
 }
 concurrencyOverrides {maxCount=2}
 soundSubmixSends [{
   minSendLevel=0.25
   soundSubmix=&Fortnite_Reverb_Default
   customSendLevelCurve {editorCurveData {keys [<at 0 0>,<at 1 1>]}}
  }]
}
